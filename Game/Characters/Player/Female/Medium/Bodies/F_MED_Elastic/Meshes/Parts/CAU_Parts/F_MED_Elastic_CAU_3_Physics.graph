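# Physics asset for the CAU_3 hair part of the female medium Elastic player body, setting up jointed hair physics.
# Capsule bodies cover the neck, head, hair root and strands, with free joints linking neck, head and root.
# Five two-segment back chains, two side strands and a front chain hang off the root, with two more side strands branching from the front.
# Strand joints have tight angular limits, mostly twist-locked at the back.
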PhysicsAsset /Game/Characters/Player/Female/Medium/Bodies/F_MED_Elastic/Meshes/Parts/CAU_Parts/F_MED_Elastic_CAU_3_Physics
bodies {
 cau3_back_hair_a [{shapeType=Capsule, length=1, radius=2, center=<2.4727 0 0>, rotation=<rot 90 0 0>}]
 cau3_back_hair_a_base [{shapeType=Capsule, length=2.5, radius=2, center=<3 0 0>, rotation=<rot 90 0 0>}]
 cau3_back_hair_b [{shapeType=Capsule, length=1, radius=2, center=<2.4727 0 0>, rotation=<rot 90 0 0>}]
 cau3_back_hair_b_base [{shapeType=Capsule, length=2.5, radius=2, center=<3 0 0>, rotation=<rot 90 0 0>}]
 cau3_back_hair_c [{shapeType=Capsule, length=1, radius=2, center=<2.4727 0 0>, rotation=<rot 90 0 0>}]
 cau3_back_hair_c_base [{shapeType=Capsule, length=2.5, radius=2, center=<3 0 0>, rotation=<rot 90 0 0>}]
 cau3_back_hair_d [{shapeType=Capsule, length=1, radius=2, center=<2.4727 0 0>, rotation=<rot 90 0 0>}]
 cau3_back_hair_d_base [{shapeType=Capsule, length=2.5, radius=2, center=<3 0 0>, rotation=<rot 90 0 0>}]
 cau3_back_hair_e [{shapeType=Capsule, length=1, radius=2, center=<2.4727 0 0>, rotation=<rot 90 0 0>}]
 cau3_back_hair_e_base [{shapeType=Capsule, length=2.5, radius=2, center=<3 0 0>, rotation=<rot 90 0 0>}]
 cau3_front_hair_a [{shapeType=Capsule, length=1.75, radius=1, center=<1.6295 0 0>, rotation=<rot 90 0 0>}]
 cau3_front_hair_b [{shapeType=Capsule, length=2, radius=0.75, center=<1.7581 0 0>, rotation=<rot 90 0 0>}]
 cau3_front_hair_c [{shapeType=Capsule, length=3.75, radius=0.75, center=<2.689 0.3161 -0.1011>, rotation=<rot 88.3771 -171.8736 -179.9999>}]
 cau3_hair [{shapeType=Capsule, length=1, radius=1}]
 cau3_side_hair_a [{shapeType=Capsule, length=1.75, radius=0.75, center=<1.2515 0 0>, rotation=<rot 90 0 0>}]
 cau3_side_hair_b [{shapeType=Capsule, length=2, radius=1, center=<1.1707 0 0>, rotation=<rot 90 0 0>}]
 cau3_side_hair_c [{shapeType=Capsule, length=3.5, radius=0.75, center=<2.2975 0 0>, rotation=<rot 90 0 0>}]
 cau3_side_hair_d [{shapeType=Capsule, length=1, radius=1, center=<0.1441 0.8441 0>, rotation=<rot 90 5.7106 -74.6012>}]
 head [{shapeType=Capsule, length=5.25, radius=6.25, center=<2.8929 2.4756 -0.111>, rotation=<rot 90 0 0>}]
 neck_02 [{shapeType=Capsule, length=3.5, radius=4, center=<0 0.2988 0>, rotation=<rot 90 0 0>}]
}
constraints [
 {
  bone1Name=head
  bone2Name=neck_02
  swing1LimitDegrees=45
  swing1Motion=Free
  swing2LimitDegrees=45
  swing2Motion=Free
  twistLimitDegrees=45
  twistMotion=Free
 }
 {
  bone1Name=cau3_hair
  bone2Name=head
  swing1LimitDegrees=45
  swing1Motion=Free
  swing2LimitDegrees=45
  swing2Motion=Free
  twistLimitDegrees=45
  twistMotion=Free
 }
 {bone1Name=cau3_back_hair_a_base, bone2Name=cau3_hair, swing1LimitDegrees=5, swing2LimitDegrees=5, twistLimitDegrees=5, twistMotion=Locked}
 {
  bone1Name=cau3_back_hair_a
  bone2Name=cau3_back_hair_a_base
  swing1LimitDegrees=10
  swing2LimitDegrees=10
  twistLimitDegrees=5
  twistMotion=Locked
 }
 {bone1Name=cau3_back_hair_b_base, bone2Name=cau3_hair, swing1LimitDegrees=5, swing2LimitDegrees=5, twistLimitDegrees=5, twistMotion=Locked}
 {
  bone1Name=cau3_back_hair_b
  bone2Name=cau3_back_hair_b_base
  swing1LimitDegrees=5
  swing2LimitDegrees=15
  twistLimitDegrees=5
  twistMotion=Locked
 }
 {bone1Name=cau3_back_hair_c_base, bone2Name=cau3_hair, swing1LimitDegrees=5, swing2LimitDegrees=5, twistLimitDegrees=5, twistMotion=Locked}
 {
  bone1Name=cau3_back_hair_c
  bone2Name=cau3_back_hair_c_base
  swing1LimitDegrees=5
  swing2LimitDegrees=15
  twistLimitDegrees=5
  twistMotion=Locked
 }
 {bone1Name=cau3_back_hair_d_base, bone2Name=cau3_hair, swing1LimitDegrees=5, swing2LimitDegrees=5, twistLimitDegrees=5, twistMotion=Locked}
 {bone1Name=cau3_back_hair_d, bone2Name=cau3_back_hair_d_base, swing1LimitDegrees=5, swing2LimitDegrees=15, twistLimitDegrees=5}
 {bone1Name=cau3_back_hair_e_base, bone2Name=cau3_hair, swing1LimitDegrees=5, swing2LimitDegrees=5, twistLimitDegrees=5, twistMotion=Locked}
 {
  bone1Name=cau3_back_hair_e
  bone2Name=cau3_back_hair_e_base
  swing1LimitDegrees=10
  swing2LimitDegrees=10
  twistLimitDegrees=5
  twistMotion=Locked
 }
 {bone1Name=cau3_side_hair_c, bone2Name=cau3_hair, swing1LimitDegrees=5, swing2LimitDegrees=5, twistLimitDegrees=5}
 {bone1Name=cau3_side_hair_d, bone2Name=cau3_hair, swing1LimitDegrees=15, swing2LimitDegrees=15, twistLimitDegrees=5}
 {bone1Name=cau3_front_hair_a, bone2Name=cau3_hair, swing1LimitDegrees=10, swing2LimitDegrees=10, twistLimitDegrees=5}
 {bone1Name=cau3_front_hair_b, bone2Name=cau3_front_hair_a, swing1LimitDegrees=10, swing2LimitDegrees=10, twistLimitDegrees=5}
 {bone1Name=cau3_front_hair_c, bone2Name=cau3_front_hair_b, swing1LimitDegrees=10, swing2LimitDegrees=10, twistLimitDegrees=5}
 {bone1Name=cau3_side_hair_b, bone2Name=cau3_front_hair_a, swing1LimitDegrees=10, swing2LimitDegrees=10, twistLimitDegrees=5}
 {bone1Name=cau3_side_hair_a, bone2Name=cau3_front_hair_a, swing1LimitDegrees=15, swing2LimitDegrees=15, twistLimitDegrees=5}
]
